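# Physics asset for the Weeping Woods Spring backpack cosmetic, giving its flowers, leaves and honey drips constrained secondary motion.
# Every body is a single box, and a tiny offset box on the pack bone serves as the root.
# Seven two-segment leaf chains hang from the pack with tight swing limits, slightly looser at most tips.
# Six flowers and three honey drips attach directly to the pack, with the drips allowed the widest swing and twist.
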PhysicsAsset /BRCosmetics/Accessories/FORT_Backpacks/Backpack_M_MED_Weeping_Woods_Spring/Meshes/M_MED_Weeping_Woods_Spring_Pack_Physics
bodies {
 flower1 [{shapeType=Box, extentX=2.3759, extentY=6.2095, extentZ=6.6123, center=<0.7151 -0.1919 -0.0099>}]
 flower2 [{shapeType=Box, extentX=2.9134, extentY=4.9238, extentZ=6.3921, center=<0.7341 -0.6575 0.2715>}]
 flower3 [{shapeType=Box, extentX=1.9736, extentY=6.2531, extentZ=6.4153, center=<0.4805 -0.2131 0.3017>}]
 flower4 [{shapeType=Box, extentX=2.7029, extentY=8.0771, extentZ=8.4973, center=<0.1938 -0.3229 -0.4146>}]
 flower5 [{shapeType=Box, extentX=1.8862, extentY=3.4054, extentZ=5.3698, center=<0.5483 -0.5354 0.0963>}]
 flower6 [{shapeType=Box, extentX=2.1296, extentY=4.9648, extentZ=7.0744, center=<0.9684 -0.8338 0.2734>}]
 honeyDrip1 [{shapeType=Box, extentX=2.3499, extentY=1.8541, extentZ=1.9198, center=<1.3774 -0.1075 0.1313>}]
 honeyDrip2 [{shapeType=Box, extentX=3.1335, extentY=2.1267, extentZ=2.0007, center=<2.2557 0.1053 0.0612>}]
 honeyDrip3 [{shapeType=Box, extentX=1.9821, extentY=1.172, extentZ=1.3154, center=<1.7167 0.0962 0.0195>}]
 leafA1 [{shapeType=Box, extentX=2.5, extentY=1, extentZ=3, center=<1.4143 0.0001 0>}]
 leafA2 [{shapeType=Box, extentX=3, extentY=1, extentZ=3, center=<1.5995 0 0>}]
 leafB1 [{shapeType=Box, extentX=3, extentY=1, extentZ=3, center=<1.5 0 0>}]
 leafB2 [{shapeType=Box, extentX=3, extentY=1, extentZ=3, center=<1.5 0 0>}]
 leafC1 [{shapeType=Box, extentX=1.5, extentY=1, extentZ=3, center=<0.5927 0 0>}]
 leafC2 [{shapeType=Box, extentX=1.5, extentY=1, extentZ=3, center=<0.7224 0 0>}]
 leafD1 [{shapeType=Box, extentX=2.5, extentY=1, extentZ=3, center=<1.2532 0 0>}]
 leafD2 [{shapeType=Box, extentX=3, extentY=1, extentZ=3, center=<1.5 0 0>}]
 leafE1 [{shapeType=Box, extentX=2.5, extentY=1, extentZ=3, center=<1.1465 0 0>}]
 leafE2 [{shapeType=Box, extentX=3, extentY=1, extentZ=3, center=<1.5 0 0>}]
 leafF1 [{shapeType=Box, extentX=3, extentY=1, extentZ=3, center=<1.5 0 0>}]
 leafF2 [{shapeType=Box, extentX=3, extentY=1, extentZ=3, center=<1.5 0 0>}]
 leafG1 [{shapeType=Box, extentX=3, extentY=1, extentZ=3, center=<1.5 0 0>}]
 leafG2 [{shapeType=Box, extentX=3, extentY=1, extentZ=3, center=<1.5 0 0>}]
 pack [{shapeType=Box, extentX=0.5, extentY=0.5, extentZ=0.5, center=<-0.417 -10.0342 -7.9924>}]
}
constraints [
 {bone1Name=leafA1, bone2Name=pack, swing1LimitDegrees=5, swing2LimitDegrees=5, twistLimitDegrees=1}
 {bone1Name=leafA2, bone2Name=leafA1, swing1LimitDegrees=8, swing2LimitDegrees=5, twistLimitDegrees=1}
 {bone1Name=leafB1, bone2Name=pack, swing1LimitDegrees=5, swing2LimitDegrees=5, twistLimitDegrees=1}
 {bone1Name=leafB2, bone2Name=leafB1, swing1LimitDegrees=12, swing2LimitDegrees=5, twistLimitDegrees=1}
 {bone1Name=leafC1, bone2Name=pack, swing1LimitDegrees=5, swing2LimitDegrees=5, twistLimitDegrees=1}
 {bone1Name=leafC2, bone2Name=leafC1, swing1LimitDegrees=10, swing2LimitDegrees=5, twistLimitDegrees=1}
 {bone1Name=leafD1, bone2Name=pack, swing1LimitDegrees=5, swing2LimitDegrees=5, twistLimitDegrees=1}
 {bone1Name=leafD2, bone2Name=leafD1, swing1LimitDegrees=8, swing2LimitDegrees=5, twistLimitDegrees=1}
 {bone1Name=leafE1, bone2Name=pack, swing1LimitDegrees=5, swing2LimitDegrees=5, twistLimitDegrees=1}
 {bone1Name=leafE2, bone2Name=leafE1, swing1LimitDegrees=11, swing2LimitDegrees=5, twistLimitDegrees=1}
 {bone1Name=leafF1, bone2Name=pack, swing1LimitDegrees=5, swing2LimitDegrees=5, twistLimitDegrees=1}
 {bone1Name=leafF2, bone2Name=leafF1, swing1LimitDegrees=9, swing2LimitDegrees=5, twistLimitDegrees=1}
 {bone1Name=leafG1, bone2Name=pack, swing1LimitDegrees=5, swing2LimitDegrees=5, twistLimitDegrees=1}
 {bone1Name=leafG2, bone2Name=leafG1, swing1LimitDegrees=11, swing2LimitDegrees=5, twistLimitDegrees=1}
 {bone1Name=honeyDrip1, bone2Name=pack, swing1LimitDegrees=45, swing2LimitDegrees=45, twistLimitDegrees=15}
 {bone1Name=honeyDrip2, bone2Name=pack, swing1LimitDegrees=45, swing2LimitDegrees=45, twistLimitDegrees=15}
 {bone1Name=honeyDrip3, bone2Name=pack, swing1LimitDegrees=45, swing2LimitDegrees=45, twistLimitDegrees=15}
 {bone1Name=flower1, bone2Name=pack, swing1LimitDegrees=10, swing2LimitDegrees=10, twistLimitDegrees=1}
 {bone1Name=flower2, bone2Name=pack, swing1LimitDegrees=10, swing2LimitDegrees=10, twistLimitDegrees=1}
 {bone1Name=flower3, bone2Name=pack, swing1LimitDegrees=10, swing2LimitDegrees=10, twistLimitDegrees=1}
 {bone1Name=flower4, bone2Name=pack, swing1LimitDegrees=10, swing2LimitDegrees=10, twistLimitDegrees=1}
 {bone1Name=flower5, bone2Name=pack, swing1LimitDegrees=10, swing2LimitDegrees=10, twistLimitDegrees=1}
 {bone1Name=flower6, bone2Name=pack, swing1LimitDegrees=10, swing2LimitDegrees=10, twistLimitDegrees=1}
]
